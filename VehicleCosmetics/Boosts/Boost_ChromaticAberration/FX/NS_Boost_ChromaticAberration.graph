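# Niagara system for the Chromatic Aberration vehicle boost, trailing layered ribbons from the car's filtered sockets.
# Two parent-derived emitters of different widths stack three and four ribbon renderers, with backward velocity and lifetime switched by game mode.
# Painted-variant overrides give R1 (and R4 on the wider emitter) age-based color curves, while R2 and R3 stay teal and red across paints.
# Inherited source movement, partly curve-faded, shifts particles with the vehicle.
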NiagaraSystem /VehicleCosmetics/Boosts/Boost_ChromaticAberration/FX/NS_Boost_ChromaticAberration
dependencies {
 usedDataInterfaces [
  &NiagaraDataInterfaceCurve
  &NiagaraDataInterfaceColorCurve
  &NiagaraDataInterfaceSkeletalMesh
  &NiagaraDataInterfaceArrayFloat3
 ]
 usedDynamicInputs [&SelectVectorFromArray]
 usedModules [
  &self:SystemSpawnScript.SystemScriptSource.SystemScriptGraph.NiagaraNodeAssignment_0.SetVariables_#1
  &self:SystemSpawnScript.SystemScriptSource.SystemScriptGraph.NiagaraNodeAssignment_1.SetVariables_#2
  &self:SystemSpawnScript.SystemScriptSource.SystemScriptGraph.NiagaraNodeAssignment_2.SetVariables_#3
  &SystemState
 ]
}
emitters {
 NE_VC_Boost_Ribbon_Parent {
  emitterUpdateScript [
   &EmitterState()
   &VC_FilteredSocketCount()
   &VC_RateQualityHandler(Epic Spawn Rate=18, High Spawn Rate=18, Low Spawn Rate=18, Medium Spawn Rate=18, Use Custom Rate?=true)
   &VC_SpawnRibbon_v1(Number of Ribbons=$Emitter.PortCount, Spawn Time=-1, SpawnRate=$Emitter.Spawn Rate)
  ]
  particleSpawnScript [
   &VC_PaintedVariantsOverride()
   &VC_InitializeRibbon_v1(
    Custom Offset=true
    Enable Skeletal Position=true
    Lifetime=&MergeHelper_FloatSwitch(0=1.5, 1=1.5, 2=1.5, 3=?, Index=$System.GameModeID, Num=::NewEnumerator1)
    Local Vector Offset=&AddVectorToPosition(Position=<-38 0 0>, Vector=$User.DTCarOffset)
    Number of Bones Per Boneflow group=2
    Offset Amount=1
    Ribbon Width=75
   )
   &AddVelocity(Velocity=&MergeHelper_VectorSwitch(0=<-1000 0 0>, 1=<-1100 0 0>, 2=<-1000 0 0>, 3=?, Index=$System.GameModeID, Num=::NewEnumerator1)) {module=AddVelocity001}
   /VehicleCosmetics/FX/VC_FXParents/Modules/DLR_ProductPreviews(GameModeID:NiagaraInt32=$System.GameModeID, bPreviewVehicleSpeed:NiagaraBool=$System.bPreviewVehicleSpeed)
  ]
  particleUpdateScript [
   &ParticleState()
   &InheritSourceMovement(
    Applied Position Scale=&VectorFromFloat(Value=&FloatFromCurve(FloatCurve=DataInterface() {propertyValues {Curve {
         defaultValue=1
         keys [
          {arriveTangent=-1.7885, interpMode=RCIM_Cubic, leaveTangent=-1.7885, tangentMode=RCTM_User, time=0.05, value=1}
          {interpMode=RCIM_Cubic, tangentMode=RCTM_User, time=1, value=0}
         ]
        }}}))
    ApplyForce=false
    ApplyPosition=true
    ApplyPositionFalloff=false
   )
   &InheritSourceMovement(
    Applied Position Scale=&Multiply_VectorByFloat(Vector=<1 1 1>, Float=1)
    ApplyForce=false
    ApplyPosition=true
    ApplyPositionFalloff=false
   ) {module=InheritSourceMovement002}
   &SolveForcesAndVelocity(Clamp Velocity=false, Limit Acceleration=false)
   &VC_PaintedVariantsOverride(
    Orange=<rgba 0.6724 0.0883 0.2836 1>
    R1 Black:LinearColor=&ColorFromCurve(ColorCurve=DataInterface() {propertyValues {
       AlphaCurve {keys [<at 0 4 RCIM_Cubic>,<at 0.1045 0.9695>,<at 1 0 RCIM_Cubic>]}
       BlueCurve {keys [<at 0 0.0356 RCIM_Cubic>,<at 0.6713 0.0208>]}
       GreenCurve {keys [<at 0 0.0356 RCIM_Cubic>,<at 0.6713 0.0208>]}
       RedCurve {keys [<at 0 0.0356 RCIM_Cubic>,<at 0.6713 0.0208>]}
      }})
    R1 BurntSienna:LinearColor=&ColorFromCurve(ColorCurve=DataInterface() {propertyValues {
       AlphaCurve {keys [<at 0 4 RCIM_Cubic>,<at 0.2166 2>,<at 1 0 RCIM_Cubic>]}
       BlueCurve {keys [<at 0 0.0284>,<at 0.6066 0.0061>]}
       GreenCurve {keys [<at 0 0.0717>,<at 0.6066 0.0193>]}
       RedCurve {keys [<at 0 0.2292>,<at 0.6066 0.067>]}
      }})
    R1 Cobalt:LinearColor=&ColorFromCurve(ColorCurve=DataInterface() {propertyValues {
       AlphaCurve {keys [<at 0 4 RCIM_Cubic>,<at 0.6217 1>,<at 1 0 RCIM_Cubic>]}
       BlueCurve {keys [<at 0 0.6867 RCIM_Cubic>,<at 0.6282 0.0666>]}
       GreenCurve {keys [<at 0 0.0551 RCIM_Cubic>,<at 0.6282 0.0012>]}
       RedCurve {keys [<at 0 0.0429 RCIM_Cubic>,<at 0.6282 0>]}
      }})
    R1 Crimson:LinearColor=&ColorFromCurve(ColorCurve=DataInterface() {propertyValues {
       AlphaCurve {keys [<at 0 4 RCIM_Cubic>,<at 1 0 RCIM_Cubic>]}
       BlueCurve {keys [<at 0.0162 0.0673>,<at 0.5441 0.0072>]}
       GreenCurve {keys [<at 0.0162 0.0548>,<at 0.5441 0>]}
       RedCurve {keys [<at 0.0162 0.526>,<at 0.5441 0.2708>]}
      }})
    R1 DefaultAlpha=0.9
    R1 DefaultColor=&ColorFromCurve(ColorCurve=DataInterface() {propertyValues {
       AlphaCurve {keys [<at 0 4 RCIM_Cubic>,<at 0.8546 0.9976>,<at 1 0.248 RCIM_Cubic>]}
       BlueCurve {keys [<at 0.2825 1>,<at 0.4849 0.4037>,<at 0.631 1>,<at 0.7138 0.1391>]}
       GreenCurve {keys [<at 0.2825 0.947>,<at 0.4849 1>,<at 0.631 0.9298>,<at 0.7138 0.0353>]}
       RedCurve {keys [<at 0.2825 0.9293>,<at 0.4849 0.0023>,<at 0.631 0>,<at 0.7138 0>]}
      }})
    R1 ForestGreen:LinearColor=&ColorFromCurve(ColorCurve=DataInterface() {propertyValues {
       AlphaCurve {keys [<at 0 4 RCIM_Cubic>,<at 1 0 RCIM_Cubic>]}
       BlueCurve {keys [<at 0.0075 0.0723>,<at 0.5485 0.0091>]}
       GreenCurve {keys [<at 0.0075 0.3419>,<at 0.5485 0.067>]}
       RedCurve {keys [<at 0.0075 0.0578>,<at 0.5485 0.0061>]}
      }})
    R1 Grey:LinearColor=&ColorFromCurve(ColorCurve=DataInterface() {propertyValues {
       AlphaCurve {keys [<at 0 1 RCIM_Cubic>,<at 0.8868 0.9>,<at 0.9795 0 RCIM_Cubic>]}
       BlueCurve {keys [<at 0 0.2527>,<at 0.5743 0.067>]}
       GreenCurve {keys [<at 0 0.2527>,<at 0.5743 0.067>]}
       RedCurve {keys [<at 0 0.2527>,<at 0.5743 0.067>]}
      }})
    R1 Lime:LinearColor=&ColorFromCurve(ColorCurve=DataInterface() {propertyValues {
       AlphaCurve {keys [<at 0 1 RCIM_Cubic>,<at 0.8997 0.9>,<at 1 0 RCIM_Cubic>]}
       BlueCurve {keys [<at 0.014 0.0508>,<at 0.611 0.0331>]}
       GreenCurve {keys [<at 0.014 0.6435>,<at 0.611 0.3613>]}
       RedCurve {keys [<at 0.014 0.344>,<at 0.611 0.1946>]}
      }})
    R1 Orange:LinearColor=&ColorFromCurve(ColorCurve=DataInterface() {propertyValues {
       AlphaCurve {keys [<at 0 1 RCIM_Cubic>,<at 0.9213 0.986>,<at 1 0 RCIM_Cubic>]}
       BlueCurve {keys [<at 0 0.0705 RCIM_Cubic>,<at 0.6217 0.0075>]}
       GreenCurve {keys [<at 0 0.357 RCIM_Cubic>,<at 0.6217 0.1329>]}
       RedCurve {keys [<at 0 0.892 RCIM_Cubic>,<at 0.6217 0.3712>]}
      }})
    R1 Pink:LinearColor=&ColorFromCurve(ColorCurve=DataInterface() {propertyValues {
       AlphaCurve {keys [<at 0 1 RCIM_Cubic>,<at 0.9277 0.9949>,<at 1 0 RCIM_Cubic>]}
       BlueCurve {keys [<at 0.0334 0.8688>,<at 0.6282 0.2874>]}
       GreenCurve {keys [<at 0.0334 0.1886>,<at 0.6282 0.0395>]}
       RedCurve {keys [<at 0.0334 0.9531>,<at 0.6282 0.4342>]}
      }})
    R1 Purple:LinearColor=&ColorFromCurve(ColorCurve=DataInterface() {propertyValues {
       AlphaCurve {keys [<at 0 1 RCIM_Cubic>,<at 0.8976 0.9795>,<at 1 0 RCIM_Cubic>]}
       BlueCurve {keys [<at 0.042 0.552>,<at 0.6239 0.1046>]}
       GreenCurve {keys [<at 0.042 0.0436>,<at 0.6239 0.0097>]}
       RedCurve {keys [<at 0.042 0.2523>,<at 0.6239 0.0482>]}
      }})
    R1 Saffron:LinearColor=&ColorFromCurve(ColorCurve=DataInterface() {propertyValues {
       AlphaCurve {keys [<at 0 1 RCIM_Cubic>,<at 0.9083 0.9979>,<at 1 0 RCIM_Cubic>]}
       BlueCurve {keys [<at 0.0075 0.0752>,<at 0.6174 0.0171>]}
       GreenCurve {keys [<at 0.0075 0.747>,<at 0.6174 0.2554>]}
       RedCurve {keys [<at 0.0075 0.952>,<at 0.6174 0.3281>]}
      }})
    R1 SkyBlue:LinearColor=&ColorFromCurve(ColorCurve=DataInterface() {propertyValues {
       AlphaCurve {keys [<at 0 1 RCIM_Cubic>,<at 0.9428 0.9815>,<at 1 0 RCIM_Cubic>]}
       BlueCurve {keys [<at 0 0.552>,<at 0.6088 0.2961>]}
       GreenCurve {keys [<at 0 0.3834>,<at 0.6088 0.2051>]}
       RedCurve {keys [<at 0 0.0436>,<at 0.6088 0.011>]}
      }})
    R1 TitaniumWhite:LinearColor=&ColorFromCurve(ColorCurve=DataInterface() {propertyValues {
       AlphaCurve {keys [<at 0 1 RCIM_Cubic>,<at 1 0 RCIM_Cubic>]}
       BlueCurve {keys [<at 0 0.9271>,<at 0.5808 0.6823>]}
       GreenCurve {keys [<at 0 0.9271>,<at 0.5808 0.6823>]}
       RedCurve {keys [<at 0 0.9271>,<at 0.5808 0.6823>]}
      }})
    R2 Black:LinearColor=<rgba 0 0.5972 0.3763 1>
    R2 BurntSienna:LinearColor=<rgba 0 0.5972 0.3763 1>
    R2 Cobalt:LinearColor=<rgba 0 0.5972 0.3763 1>
    R2 Crimson:LinearColor=<rgba 0 0.5972 0.3763 1>
    R2 DefaultAlpha=0.8
    R2 DefaultColor:LinearColor=&ColorFromCurve(ColorCurve=DataInterface() {propertyValues {
       AlphaCurve {keys [<at 0.0449 3>,<at 0.4797 3>,<at 1 3>]}
       BlueCurve {keys [<at 0.4885 0.3763>]}
       GreenCurve {keys [<at 0.4885 0.5972>]}
       RedCurve {keys [<at 0.4885 0>]}
      }})
    R2 ForestGreen:LinearColor=<rgba 0 0.5972 0.3763 1>
    R2 Grey:LinearColor=<rgba 0 0.5972 0.3763 1>
    R2 Lime:LinearColor=<rgba 0 0.5972 0.3763 1>
    R2 Orange:LinearColor=<rgba 0 0.5972 0.3763 1>
    R2 Pink:LinearColor=<rgba 0 0.5972 0.3763 1>
    R2 Purple:LinearColor=<rgba 0 0.5972 0.3763 1>
    R2 Saffron:LinearColor=<rgba 0 0.5972 0.3763 1>
    R2 SkyBlue:LinearColor=<rgba 0 0.5972 0.3763 1>
    R2 TitaniumWhite:LinearColor=<rgba 0 0.5972 0.3763 1>
    R3 Black:LinearColor=<rgba 0.3095 0 0.0097 0.8196>
    R3 BurntSienna:LinearColor=<rgba 0.3095 0 0.0097 0.8196>
    R3 Cobalt:LinearColor=<rgba 0.3095 0 0.0097 0.8196>
    R3 Crimson:LinearColor=<rgba 0.3095 0 0.0097 0.8196>
    R3 DefaultAlpha=4
    R3 DefaultColor:LinearColor=<rgba 0.3095 0 0.0097 0.8196>
    R3 ForestGreen:LinearColor=<rgba 0.3095 0 0.0097 0.8196>
    R3 Grey:LinearColor=<rgba 0.3095 0 0.0097 0.8196>
    R3 Lime:LinearColor=<rgba 0.3095 0 0.0097 0.8196>
    R3 Orange:LinearColor=<rgba 0.3095 0 0.0097 0.8196>
    R3 Pink:LinearColor=<rgba 0.3095 0 0.0097 0.8196>
    R3 Purple:LinearColor=<rgba 0.3095 0 0.0097 0.8196>
    R3 Saffron:LinearColor=<rgba 0.3095 0 0.0097 0.8196>
    R3 SkyBlue:LinearColor=<rgba 0.3095 0 0.0097 0.8196>
    R3 TitaniumWhite:LinearColor=<rgba 0.3095 0 0.0097 0.8196>
    Use R1 Painted Overrides=true
    Use R2 Painted Overrides=true
    Use R3 Painted Overrides=true
   ) {module=VC_PaintedVariantsOverride001}
   &VC_UpdateRibbon_v1(Enable Skeletal Position=true)
  ]
  rendererClasses [&NiagaraRibbonRendererProperties]
  renderers [
   {rendererClass=&NiagaraRibbonRendererProperties}
   {rendererClass=&NiagaraRibbonRendererProperties, rendererIndex=1}
   {rendererClass=&NiagaraRibbonRendererProperties, rendererIndex=2}
  ]
 }
 NE_VC_Boost_Ribbon_Parent001 {
  emitterUpdateScript [
   &EmitterState()
   &VC_FilteredSocketCount()
   &VC_RateQualityHandler(Epic Spawn Rate=18, High Spawn Rate=18, Low Spawn Rate=18, Medium Spawn Rate=18, Use Custom Rate?=true)
   &VC_SpawnRibbon_v1(Number of Ribbons=$Emitter.PortCount, Spawn Time=-1, SpawnRate=$Emitter.Spawn Rate)
  ]
  particleSpawnScript [
   &VC_PaintedVariantsOverride()
   &VC_InitializeRibbon_v1(
    Custom Offset=true
    Enable Skeletal Position=true
    Lifetime=&MergeHelper_FloatSwitch(0=1.5, 1=1.5, 2=1.5, 3=?, Index=$System.GameModeID, Num=::NewEnumerator1)
    Local Vector Offset=&AddVectorToPosition(Position=<-38 0 0>, Vector=$User.DTCarOffset)
    Number of Bones Per Boneflow group=2
    Offset Amount=1
    Ribbon Width=100
   )
   &AddVelocity(Velocity=&MergeHelper_VectorSwitch(0=<-1000 0 0>, 1=<-1100 0 0>, 2=<-1000 0 0>, 3=?, Index=$System.GameModeID, Num=::NewEnumerator1)) {module=AddVelocity001}
   &DLR_ProductPreviews(GameModeID:NiagaraInt32=$System.GameModeID, bPreviewVehicleSpeed:NiagaraBool=$System.bPreviewVehicleSpeed)
  ]
  particleUpdateScript [
   &ParticleState()
   &InheritSourceMovement(
    Applied Position Scale=&VectorFromFloat(Value=&FloatFromCurve(FloatCurve=DataInterface() {propertyValues {Curve {
         defaultValue=1
         keys [
          {arriveTangent=-1.7885, interpMode=RCIM_Cubic, leaveTangent=-1.7885, tangentMode=RCTM_User, time=0.05, value=1}
          {interpMode=RCIM_Cubic, tangentMode=RCTM_User, time=1, value=0}
         ]
        }}}))
    ApplyForce=false
    ApplyPosition=true
    ApplyPositionFalloff=false
   )
   &InheritSourceMovement(
    Applied Position Scale=&Multiply_VectorByFloat(Vector=<1 1 1>, Float=1)
    ApplyForce=false
    ApplyPosition=true
    ApplyPositionFalloff=false
   ) {module=InheritSourceMovement002}
   &SolveForcesAndVelocity(Clamp Velocity=false, Limit Acceleration=false)
   &VC_PaintedVariantsOverride(
    Black=<rgba 0.1458 0.1458 0.1458 1>
    Orange=<rgba 0.6724 0.0883 0.2836 1>
    R1 Black:LinearColor=&ColorFromCurve(ColorCurve=DataInterface() {propertyValues {
       AlphaCurve {keys [<at 0 4 RCIM_Cubic>,<at 0.9444 0.8 RCIM_Cubic>]}
       BlueCurve {keys [<at 0 0.3021 RCIM_Cubic>,<at 0.9338 0.0723>]}
       GreenCurve {keys [<at 0 0.3021 RCIM_Cubic>,<at 0.9338 0.0723>]}
       RedCurve {keys [<at 0 0.3021 RCIM_Cubic>,<at 0.9338 0.0723>]}
      }})
    R1 BurntSienna:LinearColor=&ColorFromCurve(ColorCurve=DataInterface() {propertyValues {
       AlphaCurve {keys [<at 0 4 RCIM_Cubic>,<at 0.2166 2>,<at 0.9303 0.5 RCIM_Cubic>]}
       BlueCurve {keys [<at 0 0>,<at 0.6066 0>]}
       GreenCurve {keys [<at 0 0.2104>,<at 0.6066 0.1046>]}
       RedCurve {keys [<at 0 0.3385>,<at 0.6066 0.1651>]}
      }})
    R1 Cobalt:LinearColor=&ColorFromCurve(ColorCurve=DataInterface() {propertyValues {
       AlphaCurve {keys [<at 0 4 RCIM_Cubic>,<at 0.7018 0.8 RCIM_Cubic>]}
       BlueCurve {keys [<at 0 0.901 RCIM_Cubic>,<at 0.7018 0.4062>]}
       GreenCurve {keys [<at 0 0.0493 RCIM_Cubic>,<at 0.7018 0.0077>]}
       RedCurve {keys [<at 0 0.0329 RCIM_Cubic>,<at 0.7018 0>]}
      }})
    R1 Crimson:LinearColor=&ColorFromCurve(ColorCurve=DataInterface() {propertyValues {
       AlphaCurve {keys [<at 0 4 RCIM_Cubic>,<at 0.9109 0.5 RCIM_Cubic>]}
       BlueCurve {keys [<at 0 0.0072>,<at 0.624 0.0369>]}
       GreenCurve {keys [<at 0 0.0069>,<at 0.624 0.0257>]}
       RedCurve {keys [<at 0 0.901>,<at 0.624 0.4479>]}
      }})
    R1 DefaultAlpha=0.9
    R1 DefaultColor=&ColorFromCurve(ColorCurve=DataInterface() {propertyValues {
       AlphaCurve {keys [<at 0 4 RCIM_Cubic>,<at 0.8423 0.9976>,<at 1 0 RCIM_Cubic>]}
       BlueCurve {keys [<at 0.2544 1>,<at 0.455 1>,<at 0.6592 0.5729>,<at 0.8458 0.625>]}
       GreenCurve {keys [<at 0.2544 0.947>,<at 0.455 0.9298>,<at 0.6592 0.3571>,<at 0.8458 0.1586>]}
       RedCurve {keys [<at 0.2544 0.9293>,<at 0.455 0>,<at 0.6592 0.0027>,<at 0.8458 0>]}
      }})
    R1 ForestGreen:LinearColor=&ColorFromCurve(ColorCurve=DataInterface() {propertyValues {
       AlphaCurve {keys [<at 0 4 RCIM_Cubic>,<at 0.9039 0.5>]}
       BlueCurve {keys [<at 0.0075 0.0887>,<at 0.617 0.0423>]}
       GreenCurve {keys [<at 0.0075 0.855>,<at 0.617 0.3095>]}
       RedCurve {keys [<at 0.0075 0.0497>,<at 0.617 0.0284>]}
      }})
    R1 Grey:LinearColor=&ColorFromCurve(ColorCurve=DataInterface() {propertyValues {
       AlphaCurve {keys [<at 0 4 RCIM_Cubic>,<at 0.9092 0.5 RCIM_Cubic>]}
       BlueCurve {keys [<at 0 0.4896>,<at 0.5743 0.224>]}
       GreenCurve {keys [<at 0 0.4896>,<at 0.5743 0.224>]}
       RedCurve {keys [<at 0 0.4896>,<at 0.5743 0.224>]}
      }})
    R1 Lime:LinearColor=&ColorFromCurve(ColorCurve=DataInterface() {propertyValues {
       AlphaCurve {keys [<at 0 4 RCIM_Cubic>,<at 0.9092 0.5 RCIM_Cubic>]}
       BlueCurve {keys [<at 0 0.0075>,<at 0.3248 0.0075>,<at 0.705 0.0027>]}
       GreenCurve {keys [<at 0 0.9911>,<at 0.3248 0.9911>,<at 0.705 0.3594>]}
       RedCurve {keys [<at 0 0.5647>,<at 0.3248 0.5647>,<at 0.705 0.2048>]}
      }})
    R1 Orange:LinearColor=&ColorFromCurve(ColorCurve=DataInterface() {propertyValues {
       AlphaCurve {keys [<at 0 3 RCIM_Cubic>,<at 0.9213 0.5>]}
       BlueCurve {keys [<at 0 0.0331 RCIM_Cubic>,<at 0.7384 0.0232>]}
       GreenCurve {keys [<at 0 0.3916 RCIM_Cubic>,<at 0.7384 0.2423>]}
       RedCurve {keys [<at 0 0.9131 RCIM_Cubic>,<at 0.7384 0.8148>]}
      }})
    R1 Pink:LinearColor=&ColorFromCurve(ColorCurve=DataInterface() {propertyValues {
       AlphaCurve {keys [<at 0 3 RCIM_Cubic>,<at 0.9277 0.5>]}
       BlueCurve {keys [<at 0 0.9823>,<at 0.7226 0.4346>]}
       GreenCurve {keys [<at 0 0.2159>,<at 0.7226 0.1927>]}
       RedCurve {keys [<at 0 0.9047>,<at 0.7226 0.5781>]}
      }})
    R1 Purple:LinearColor=&ColorFromCurve(ColorCurve=DataInterface() {propertyValues {
       AlphaCurve {keys [<at 0 3 RCIM_Cubic>,<at 0.8976 0.5>]}
       BlueCurve {keys [<at 0 1>,<at 0.7243 0.4115>]}
       GreenCurve {keys [<at 0 0.0137>,<at 0.7243 0.0056>]}
       RedCurve {keys [<at 0 0.402>,<at 0.7243 0.1654>]}
      }})
    R1 Saffron:LinearColor=&ColorFromCurve(ColorCurve=DataInterface() {propertyValues {
       AlphaCurve {keys [<at 0 3 RCIM_Cubic>,<at 0.8722 0.5>]}
       BlueCurve {keys [<at 0.0075 0>,<at 0.7226 0>]}
       GreenCurve {keys [<at 0.0075 0.9647>,<at 0.7226 0.5024>]}
       RedCurve {keys [<at 0.0075 1>,<at 0.7226 0.5208>]}
      }})
    R1 SkyBlue:LinearColor=&ColorFromCurve(ColorCurve=DataInterface() {propertyValues {
       AlphaCurve {keys [<at 0 3 RCIM_Cubic>,<at 0.9004 0.5>]}
       BlueCurve {keys [<at 0 0.9688>,<at 0.7226 0.6042>]}
       GreenCurve {keys [<at 0 0.6728>,<at 0.7226 0.4106>]}
       RedCurve {keys [<at 0 0.0765>,<at 0.7226 0.0157>]}
      }})
    R1 TitaniumWhite:LinearColor=&ColorFromCurve(ColorCurve=DataInterface() {propertyValues {
       AlphaCurve {keys [<at 0 3 RCIM_Cubic>,<at 1 0.5 RCIM_Cubic>]}
       BlueCurve {keys [<at 0 1>,<at 0.5306 0.7135>,<at 0.7259 0.5781>]}
       GreenCurve {keys [<at 0 1>,<at 0.5306 0.7135>,<at 0.7259 0.5781>]}
       RedCurve {keys [<at 0 1>,<at 0.5306 0.7135>,<at 0.7259 0.5781>]}
      }})
    R2 Black:LinearColor=<rgba 0 0.7991 0.5089 1>
    R2 BurntSienna:LinearColor=<rgba 0 0.7991 0.5089 1>
    R2 Cobalt:LinearColor=<rgba 0 0.7991 0.5089 1>
    R2 Crimson:LinearColor=<rgba 0 0.7991 0.5089 1>
    R2 DefaultAlpha=0.1
    R2 DefaultColor:LinearColor=&ColorFromCurve(ColorCurve=DataInterface() {propertyValues {
       AlphaCurve {keys [<at 0.0449 3>,<at 0.4797 3>,<at 1 3>]}
       BlueCurve {keys [<at 0.4885 0.5089>]}
       GreenCurve {keys [<at 0.4885 0.7991>]}
       RedCurve {keys [<at 0.4885 0>]}
      }})
    R2 ForestGreen:LinearColor=<rgba 0 0.7991 0.5089 1>
    R2 Grey:LinearColor=<rgba 0 0.7991 0.5089 1>
    R2 Lime:LinearColor=<rgba 0 0.7991 0.5089 1>
    R2 Orange:LinearColor=<rgba 0 0.7991 0.5089 1>
    R2 Pink:LinearColor=<rgba 0 0.7991 0.5089 1>
    R2 Purple:LinearColor=<rgba 0 0.7991 0.5089 1>
    R2 Saffron:LinearColor=<rgba 0 0.7991 0.5089 1>
    R2 SkyBlue:LinearColor=<rgba 0 0.7991 0.5089 1>
    R2 TitaniumWhite:LinearColor=<rgba 0 0.7991 0.5089 1>
    R3 Black:LinearColor=<rgba 0.624 0 0.0194 0.8196>
    R3 BurntSienna:LinearColor=<rgba 0.624 0 0.0194 0.8196>
    R3 Cobalt:LinearColor=<rgba 0.624 0 0.0194 0.8196>
    R3 Crimson:LinearColor=<rgba 0.624 0 0.0194 0.8196>
    R3 DefaultAlpha=4
    R3 DefaultColor:LinearColor=<rgba 0.624 0 0.0194 0.8196>
    R3 ForestGreen:LinearColor=<rgba 0.624 0 0.0194 0.8196>
    R3 Grey:LinearColor=<rgba 0.624 0 0.0194 0.8196>
    R3 Lime:LinearColor=<rgba 0.624 0 0.0194 0.8196>
    R3 Orange:LinearColor=<rgba 0.624 0 0.0194 0.8196>
    R3 Pink:LinearColor=<rgba 0.624 0 0.0194 0.8196>
    R3 Purple:LinearColor=<rgba 0.624 0 0.0194 0.8196>
    R3 Saffron:LinearColor=<rgba 0.624 0 0.0194 0.8196>
    R3 SkyBlue:LinearColor=<rgba 0.624 0 0.0194 0.8196>
    R3 TitaniumWhite:LinearColor=<rgba 0.624 0 0.0194 0.8196>
    R4 Black:LinearColor=&ColorFromCurve(ColorCurve=DataInterface() {propertyValues {
       AlphaCurve {keys [<at 0.2984 1 RCIM_Cubic>,<at 1 0.989 RCIM_Cubic>]}
       BlueCurve {keys [<at 0.0534 0.2812 RCIM_Cubic>,<at 1 0.1094 RCIM_Cubic>]}
       GreenCurve {keys [<at 0.0534 0.2812 RCIM_Cubic>,<at 1 0.1094 RCIM_Cubic>]}
       RedCurve {keys [<at 0.0534 0.2812 RCIM_Cubic>,<at 1 0.1094 RCIM_Cubic>]}
      }})
    R4 BurntSienna:LinearColor=&ColorFromCurve(ColorCurve=DataInterface() {propertyValues {
       AlphaCurve {keys [<at 0 1 RCIM_Cubic>,<at 1 0.67 RCIM_Cubic>]}
       BlueCurve {keys [<at 0 0 RCIM_Cubic>,<at 1 0 RCIM_Cubic>]}
       GreenCurve {keys [<at 0 0.2337 RCIM_Cubic>,<at 1 0.0742 RCIM_Cubic>]}
       RedCurve {keys [<at 0 0.375 RCIM_Cubic>,<at 1 0.1779 RCIM_Cubic>]}
      }})
    R4 Cobalt:LinearColor=&ColorFromCurve(ColorCurve=DataInterface() {propertyValues {
       AlphaCurve {keys [<at 0.4427 1 RCIM_Cubic>,<at 0.9132 0.2 RCIM_Cubic>]}
       BlueCurve {keys [<at 0.1681 1 RCIM_Cubic>,<at 1 0.3073 RCIM_Cubic>]}
       GreenCurve {keys [<at 0.1681 0.0529 RCIM_Cubic>,<at 1 0.0021 RCIM_Cubic>]}
       RedCurve {keys [<at 0.1681 0.0103 RCIM_Cubic>,<at 1 0 RCIM_Cubic>]}
      }})
    R4 Crimson:LinearColor=&ColorFromCurve(ColorCurve=DataInterface() {propertyValues {
       AlphaCurve {keys [<at 0 1 RCIM_Cubic>,<at 1 1 RCIM_Cubic>]}
       BlueCurve {keys [<at 0 0.0156 RCIM_Cubic>,<at 0.7331 0.005 RCIM_Cubic>]}
       GreenCurve {keys [<at 0 0.0156 RCIM_Cubic>,<at 0.7331 0.005 RCIM_Cubic>]}
       RedCurve {keys [<at 0 1 RCIM_Cubic>,<at 0.7331 0.4844 RCIM_Cubic>]}
      }})
    R4 DefaultColor:LinearColor=&ColorFromCurve(ColorCurve=DataInterface() {propertyValues {
       AlphaCurve {keys [<at 0 1 RCIM_Cubic>,<at 1 0.74 RCIM_Cubic>]}
       BlueCurve {keys [<at 0.2632 0.5391 RCIM_Cubic>,<at 0.4445 0.9249>,<at 0.5131 0.9323>,<at 0.8845 2 RCIM_Cubic>]}
       GreenCurve {keys [<at 0.2632 1.431 RCIM_Cubic>,<at 0.4445 0.98>,<at 0.5131 0.8777>,<at 0.8845 0.3046 RCIM_Cubic>]}
       RedCurve {keys [<at 0.2632 1.5 RCIM_Cubic>,<at 0.4445 0.6584>,<at 0.5131 0>,<at 0.8845 0.0328 RCIM_Cubic>]}
      }})
    R4 ForestGreen:LinearColor=&ColorFromCurve(ColorCurve=DataInterface() {propertyValues {
       AlphaCurve {keys [<at 0 1 RCIM_Cubic>,<at 1 1 RCIM_Cubic>]}
       BlueCurve {keys [<at 0 0.1122 RCIM_Cubic>,<at 1 0.0292 RCIM_Cubic>]}
       GreenCurve {keys [<at 0 0.7969 RCIM_Cubic>,<at 1 0.1875 RCIM_Cubic>]}
       RedCurve {keys [<at 0 0 RCIM_Cubic>,<at 1 0.0033 RCIM_Cubic>]}
      }})
    R4 Grey:LinearColor=&ColorFromCurve(ColorCurve=DataInterface() {propertyValues {
       AlphaCurve {keys [<at 0 1 RCIM_Cubic>,<at 1 1 RCIM_Cubic>]}
       BlueCurve {keys [<at 0 0.4427 RCIM_Cubic>,<at 1 0.0729 RCIM_Cubic>]}
       GreenCurve {keys [<at 0 0.4427 RCIM_Cubic>,<at 1 0.0729 RCIM_Cubic>]}
       RedCurve {keys [<at 0 0.4427 RCIM_Cubic>,<at 1 0.0729 RCIM_Cubic>]}
      }})
    R4 Lime:LinearColor=&ColorFromCurve(ColorCurve=DataInterface() {propertyValues {
       AlphaCurve {keys [<at 0 1 RCIM_Cubic>,<at 1 1 RCIM_Cubic>]}
       BlueCurve {keys [<at 0 0.1604 RCIM_Cubic>,<at 1 0.0451 RCIM_Cubic>]}
       GreenCurve {keys [<at 0 1.1 RCIM_Cubic>,<at 1 0.2708 RCIM_Cubic>]}
       RedCurve {keys [<at 0 0.6901 RCIM_Cubic>,<at 1 0.1724 RCIM_Cubic>]}
      }})
    R4 Orange:LinearColor=&ColorFromCurve(ColorCurve=DataInterface() {propertyValues {
       AlphaCurve {keys [<at 0 1 RCIM_Cubic>,<at 1 1 RCIM_Cubic>]}
       BlueCurve {keys [<at 0 0.0041 RCIM_Cubic>,<at 1 0.0176 RCIM_Cubic>]}
       GreenCurve {keys [<at 0 0.3788 RCIM_Cubic>,<at 1 0.2077 RCIM_Cubic>]}
       RedCurve {keys [<at 0 0.8073 RCIM_Cubic>,<at 1 0.4844 RCIM_Cubic>]}
      }})
    R4 Pink:LinearColor=&ColorFromCurve(ColorCurve=DataInterface() {propertyValues {
       AlphaCurve {keys [<at 0 1 RCIM_Cubic>,<at 1 1 RCIM_Cubic>]}
       BlueCurve {keys [<at 0 0.7379 RCIM_Cubic>,<at 1 0.3051 RCIM_Cubic>]}
       GreenCurve {keys [<at 0 0.3231 RCIM_Cubic>,<at 1 0.0362 RCIM_Cubic>]}
       RedCurve {keys [<at 0 0.9823 RCIM_Cubic>,<at 1 0.4635 RCIM_Cubic>]}
      }})
    R4 Purple:LinearColor=&ColorFromCurve(ColorCurve=DataInterface() {propertyValues {
       AlphaCurve {keys [<at 0 1 RCIM_Cubic>,<at 1 1 RCIM_Cubic>]}
       BlueCurve {keys [<at 0 1 RCIM_Cubic>,<at 1 0.4792 RCIM_Cubic>]}
       GreenCurve {keys [<at 0 0 RCIM_Cubic>,<at 1 0 RCIM_Cubic>]}
       RedCurve {keys [<at 0 0.4125 RCIM_Cubic>,<at 1 0.1977 RCIM_Cubic>]}
      }})
    R4 Saffron:LinearColor=&ColorFromCurve(ColorCurve=DataInterface() {propertyValues {
       AlphaCurve {keys [<at 0 1 RCIM_Cubic>,<at 1 1 RCIM_Cubic>]}
       BlueCurve {keys [<at 0 0.2502 RCIM_Cubic>,<at 1 0.0256 RCIM_Cubic>]}
       GreenCurve {keys [<at 0 1 RCIM_Cubic>,<at 1 0.5469 RCIM_Cubic>]}
       RedCurve {keys [<at 0 1 RCIM_Cubic>,<at 1 0.5469 RCIM_Cubic>]}
      }})
    R4 SkyBlue:LinearColor=&ColorFromCurve(ColorCurve=DataInterface() {propertyValues {
       AlphaCurve {keys [<at 0 1 RCIM_Cubic>,<at 1 1 RCIM_Cubic>]}
       BlueCurve {keys [<at 0 1 RCIM_Cubic>,<at 1 0.4375 RCIM_Cubic>]}
       GreenCurve {keys [<at 0 0.6867 RCIM_Cubic>,<at 1 0.3004 RCIM_Cubic>]}
       RedCurve {keys [<at 0 0.0369 RCIM_Cubic>,<at 1 0.0161 RCIM_Cubic>]}
      }})
    R4 TitaniumWhite:LinearColor=&ColorFromCurve(ColorCurve=DataInterface() {propertyValues {
       AlphaCurve {keys [<at 0 1 RCIM_Cubic>,<at 1 1 RCIM_Cubic>]}
       BlueCurve {keys [<at 0 1 RCIM_Cubic>,<at 1 0.2917 RCIM_Cubic>]}
       GreenCurve {keys [<at 0 1 RCIM_Cubic>,<at 1 0.2917 RCIM_Cubic>]}
       RedCurve {keys [<at 0 1 RCIM_Cubic>,<at 1 0.2917 RCIM_Cubic>]}
      }})
    Use R1 Painted Overrides=true
    Use R2 Painted Overrides=true
    Use R3 Painted Overrides=true
    Use R4 Painted Overrides=true
   ) {module=VC_PaintedVariantsOverride001}
   &VC_UpdateRibbon_v1(Enable Skeletal Position=true)
  ]
  rendererClasses [&NiagaraRibbonRendererProperties]
  renderers [
   {rendererClass=&NiagaraRibbonRendererProperties}
   {rendererClass=&NiagaraRibbonRendererProperties, rendererIndex=1}
   {rendererClass=&NiagaraRibbonRendererProperties, rendererIndex=2}
   {rendererClass=&NiagaraRibbonRendererProperties, rendererIndex=3}
  ]
 }
}
summary {
 systemName=NS_Boost_ChromaticAberration
 emitters [
  {emitterName=NE_VC_Boost_Ribbon_Parent001, rendererClasses [&NiagaraRibbonRendererProperties]}
  {emitterName=NE_VC_Boost_Ribbon_Parent, rendererClasses [&NiagaraRibbonRendererProperties]}
 ]
 userVariables [
  User.bIsFrontendPreview {type=NiagaraBool}
  User.bIsPainted {type=NiagaraBool}
  User.blsIsCoroColor {type=NiagaraBool}
  User.bPreviewVehicleSpeed {type=NiagaraBool}
  User.ColorVariantID {type=NiagaraInt32}
  User.DTCarOffset {type=Vector3f, description="", defaultValue {struct=Vector3f, value=<0 0 0>}}
  User.GameModeID {type=NiagaraInt32}
  User.PaintedColor {type=LinearColor, description="", defaultValue {struct=LinearColor, value=<rgba 1 1 1 1>}}
  User.PaintedScalar {type=NiagaraFloat, defaultValue=1, description=""}
  User.PortCount {type=NiagaraInt32, defaultValue=1}
  User.RibbonCoreIntensity {type=NiagaraFloat, defaultValue=10, description=""}
  User.RibbonMidIntensity {type=NiagaraFloat, defaultValue=5, description=""}
  User.RibbonOutIntensity {type=NiagaraFloat, defaultValue=2, description=""}
 ]
}
